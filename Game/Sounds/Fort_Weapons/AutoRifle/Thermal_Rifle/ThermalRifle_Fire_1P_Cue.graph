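# Sound cue for the first-person firing sound of the Thermal Rifle.
# A random node picks among three equally weighted close-fire wave variants, which are stored in the M4A1 sound folder.
# Playback uses first-person weapon attenuation, a ranged-fire 1P sound class and the shared first-person gunfire concurrency set.
# The cue's debug flag is on.
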SoundCue /Game/Sounds/Fort_Weapons/AutoRifle/Thermal_Rifle/ThermalRifle_Fire_1P_Cue
root=Random_1
nodes [
 {id=Random_1, children [WavePlayer_0,WavePlayer_1,WavePlayer_6], properties {weights [1,1,1]}}
 {id=WavePlayer_0, properties {soundWaveAssetPtr=/Game/Sounds/Fort_Weapons/AutoRifle/M4A1/Weapon_ThermalRifle_FireCloseNew_01}}
 {id=WavePlayer_1, properties {soundWaveAssetPtr=/Game/Sounds/Fort_Weapons/AutoRifle/M4A1/Weapon_ThermalRifle_FireCloseNew_02}}
 {id=WavePlayer_6, properties {soundWaveAssetPtr=/Game/Sounds/Fort_Weapons/AutoRifle/M4A1/Weapon_ThermalRifle_FireCloseNew_03}}
]
properties {
 attenuationSettings=&Fort_Weapon_Projectile_FirstPerson_Athena_Att
 bDebug=true
 maxDistance=2097152
 soundClassObject=&Fort_WeaponRanged_Fire_1P_PSM
 volumeMultiplier=1
 concurrencyOverrides {bLimitToOwner=true, maxCount=3, resolutionRule=StopOldest}
 concurrencySet [&Weapon_Gunfire_FirstPerson_Conc]
}
