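# Sound cue for Thermal Rifle gunfire, using the third-person ranged fire sound class with local and global gunfire concurrency sets.
# It mixes close, medium and distant layers, each with its own attenuation and three equally weighted random waves.
# Close plays Thermal Rifle samples, while medium and distant reuse M4A1 samples with raised pitch and boosted volume.
# Medium and distant share a distant-fire sound class override and sit lower in the final mix.
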SoundCue /Game/Sounds/Fort_Weapons/AutoRifle/Thermal_Rifle/ThermalRifle_Fire_Cue
root=Mixer_1
nodes [
 {
  id=Attenuation_118
  children [Random_4]
  properties {attenuationSettings=/Game/Sounds/Attenuation/Fort_Weapon_Projectile_Close_Athena_Att_Scoped}
 }
 {
  id=Attenuation_119
  children [Modulator_0]
  properties {attenuationSettings=&Fort_Weapon_Projectile_Medium_Mono_Athena_Att}
 }
 {
  id=Attenuation_120
  children [Modulator_1]
  properties {attenuationSettings=&Fort_Weapon_Projectile_Distant_Mono_Athena_Att}
 }
 {id=Mixer_1, children [SoundClass_77, SoundClass_76], properties {inputVolume [1,0.7]}}
 {id=Mixer_4, children [Attenuation_119, Attenuation_120], properties {inputVolume [0.8,0.8]}}
 {id=Modulator_0, children [Random_8], properties {pitchMax=1.2, pitchMin=1.1, volumeMax=2, volumeMin=2}}
 {id=Modulator_1, children [Random_9], properties {pitchMax=1.2, pitchMin=1.1, volumeMax=2, volumeMin=2}}
 {id=Random_4, children [WavePlayer_12, WavePlayer_13, WavePlayer_14], properties {weights [1,1,1]}}
 {id=Random_8, children [WavePlayer_26, WavePlayer_27, WavePlayer_28], properties {weights [1,1,1]}}
 {id=Random_9, children [WavePlayer_29, WavePlayer_30, WavePlayer_31], properties {weights [1,1,1]}}
 {id=SoundClass_76, children [Mixer_4], properties {soundClassOverride=&Fort_WeaponRanged_Fire_Distant}}
 {
  id=SoundClass_77
  children [Attenuation_118]
  properties {soundClassOverride=&Fort_WeaponRanged_Fire_3P_PSM}
 }
 {id=WavePlayer_12, properties {soundWaveAssetPtr=/Game/Sounds/Fort_Weapons/AutoRifle/M4A1/Weapon_ThermalRifle_FireCloseNew_01}}
 {id=WavePlayer_13, properties {soundWaveAssetPtr=/Game/Sounds/Fort_Weapons/AutoRifle/M4A1/Weapon_ThermalRifle_FireCloseNew_02}}
 {id=WavePlayer_14, properties {soundWaveAssetPtr=/Game/Sounds/Fort_Weapons/AutoRifle/M4A1/Weapon_ThermalRifle_FireCloseNew_03}}
 {id=WavePlayer_26, properties {soundWaveAssetPtr=/Game/Sounds/Fort_Weapons/AutoRifle/M4A1/Weapon_M4A1_Fire_MediumNew_01}}
 {id=WavePlayer_27, properties {soundWaveAssetPtr=/Game/Sounds/Fort_Weapons/AutoRifle/M4A1/Weapon_M4A1_Fire_MediumNew_02}}
 {id=WavePlayer_28, properties {soundWaveAssetPtr=/Game/Sounds/Fort_Weapons/AutoRifle/M4A1/Weapon_M4A1_Fire_MediumNew_03}}
 {id=WavePlayer_29, properties {soundWaveAssetPtr=/Game/Sounds/Fort_Weapons/AutoRifle/M4A1/Weapon_M4A1_FireNew_Distant01}}
 {id=WavePlayer_30, properties {soundWaveAssetPtr=/Game/Sounds/Fort_Weapons/AutoRifle/M4A1/Weapon_M4A1_FireNew_Distant02}}
 {id=WavePlayer_31, properties {soundWaveAssetPtr=/Game/Sounds/Fort_Weapons/AutoRifle/M4A1/Weapon_M4A1_FireNew_Distant03}}
]
properties {
 bDebug=true
 maxDistance=27500
 soundClassObject=&Fort_WeaponRanged_Fire_3P_PSM
 volumeMultiplier=1
 concurrencyOverrides {bLimitToOwner=true, maxCount=3, resolutionRule=StopOldest}
 concurrencySet [&Weapon_Gunfire_Local_Conc, &Weapon_Gunfire_Global_Conc]
}
