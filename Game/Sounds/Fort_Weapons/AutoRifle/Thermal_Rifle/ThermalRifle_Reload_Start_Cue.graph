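# Sound cue for the start of the Thermal Rifle reload, playing the shared auto-rifle clip-out foley sound.
# A single wave player feeds the root directly, with no randomization or modulation.
# It routes through the weapon foley sound class with the reload attenuation preset, reduced volume and a concurrency override.
# The debug flag is left on.
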SoundCue /Game/Sounds/Fort_Weapons/AutoRifle/Thermal_Rifle/ThermalRifle_Reload_Start_Cue
root=WavePlayer_29
nodes [{id=WavePlayer_29, properties {soundWaveAssetPtr=/Game/Sounds/Fort_Weapons/AutoRifle/Reload_Elements/rifle_auto_single_clip_out}}]
properties {
 attenuationSettings=&Fort_Weapon_Reload_Att
 bDebug=true
 bOverrideConcurrency=true
 maxDistance=2250
 soundClassObject=&Fort_WeaponFoley
 volumeMultiplier=0.451
}
