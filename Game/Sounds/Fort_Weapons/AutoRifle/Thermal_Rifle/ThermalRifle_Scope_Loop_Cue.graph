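# Looping sound cue for the Thermal Rifle's scope, played at low volume in the weapon foley sound class.
# A quality-level node routes the first two tiers to the standard looping wave and the last two to the "_opt" variant.
# Overridden attenuation sets distance falloff with a non-spatialized near radius and stereo spread.
# Overridden concurrency prevents new instances once the limit is reached.
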SoundCue /Game/Sounds/Fort_Weapons/AutoRifle/Thermal_Rifle/ThermalRifle_Scope_Loop_Cue
root=QualityLevel_0
nodes [
 {id=QualityLevel_0, children [WavePlayer_1,WavePlayer_1,WavePlayer_0,WavePlayer_0]}
 {
  id=WavePlayer_0
  properties {bLooping=true, soundWaveAssetPtr=&/ThermalRifle_Scope_Loop01_opt}
 }
 {id=WavePlayer_1, properties {bLooping=true, soundWaveAssetPtr=&/ThermalRifle_Scope_Loop01}}
]
properties {
 bOverrideAttenuation=true
 bOverrideConcurrency=true
 maxDistance=2000
 soundClassObject=&Fort_WeaponFoley
 volumeMultiplier=0.175
 attenuationOverrides {falloffDistance=1500, nonSpatializedRadiusStart=400, stereoSpread=300, attenuationShapeExtents {x=500}}
 concurrencyOverrides {resolutionRule=PreventNew}
}
