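# Sound cue for the thermal rifle's scope zoom-in, classed as weapon foley.
# A single wave player feeds a Modulator left at default settings, with the cue's volume turned well down.
# Attenuation and spatialization are overridden off, so it plays flat regardless of listener position.
# A concurrency override rejects new instances once the limit is reached rather than stopping playing ones.
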SoundCue /Game/Sounds/Fort_Weapons/AutoRifle/Thermal_Rifle/ThermalRifle_Scope_ZoomIn_Cue
root=Modulator_2
nodes [
 {id=Modulator_2, children [WavePlayer_3]}
 {id=WavePlayer_3, properties {soundWaveAssetPtr=&/ThermalRifle_Scope_ZoomIn01}}
]
properties {
 bOverrideAttenuation=true
 bOverrideConcurrency=true
 maxDistance=2097152
 soundClassObject=&Fort_WeaponFoley
 volumeMultiplier=0.301
 attenuationOverrides {bAttenuate=false, bSpatialize=false, falloffDistance=1500, nonSpatializedRadiusStart=200, attenuationShapeExtents {x=1200}}
 concurrencyOverrides {resolutionRule=PreventNew}
}
